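# Translucent unlit reticle material for the Repulsor locked-on state, usable on Niagara sprites and drawn without depth testing.
# An MF_UI_Circle outline forms the inner ring.
# Rotated, radially mirrored UVs repeat an anti-aliased triangle SDF around it, offset by the TriangleSpread parameter.
# Both shapes are procedural, with a red-tinted sum for color and a scaled max for a combined mask.
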Material /WiseOtter/Assets/UI/Reticle/M_RepulsorLockedOn
graph {
 properties {
  bAutomaticallySetUsageInEditor=false
  bDisableDepthTest=true
  bUsedWithNiagaraSprites=true
  blendMode=BLEND_Translucent
  pixelDepthOffsetMode=PDOM_Legacy
  shadingModel=MSM_Unlit
  translucencyPass=MTP_AfterMotionBlur
  usedShadingModels=Unlit
 }
 nodes [
  n1 = NamedRerouteUsage()
  n2 = Constant4Vector() {constant=<rgba 1 0 0 1>}
  n3 = NamedRerouteUsage()
  Add(A=Multiply(A=@n1, B=@n2.RGBA), B=Multiply(A=@n3, B=@n2.R))
  Multiply(A=Max(A=@n1, B=@n3)) {constB=0.625}
  NamedRerouteDeclaration(None=MaterialFunctionCall(Size=Constant() {r=0.2}, OutlineThickness=Constant() {r=0.01}, GlowMax=Constant() {r=0.0075}) {desc=circle, materialFunction=&MF_UI_Circle}.Outline) {name=InnerCircle, nodeColor=<rgba 1 0.3516 0 1>}
  n4 = TextureCoordinate()
  n5 = MaterialFunctionCall(
   UVs=MaterialFunctionCall!(
    UVs=MaterialFunctionCall(UVs=@n4) {materialFunction=&CustomRotator}.Rotated Values
    Sides=Constant() {r=4}
   ) {materialFunction=/FortUILibrary/MaterialFunctions/Basics/MF_UI_UVRadialMirror}.UVs
   Texture Scale=Constant() {r=0.3}
  ) {materialFunction=&ScaleUVsByCenter}
  n6 = ComponentMask(None=@n5.UVs) {g=true}
  n7 = LinearInterpolate(Alpha=ScalarParameter() {defaultValue=-0.85, parameterName=TriangleSpread}) {constA=-0.5, constB=0}
  n8 = MaterialFunctionCall(UVs=AppendVector(A=ComponentMask(None=@n5.UVs) {r=true}, B=Add(A=@n6, B=@n7) {constB=-0.4}), "Rotation Angle (0-1)"=Divide() {constB=6}) {materialFunction=&CustomRotator}
  n9 = Constant() {r=3}
  NamedRerouteDeclaration(None=MaterialFunctionCall(DistanceField=MaterialFunctionCall!(UVs=@n8.Rotated Values, Radius, Sides=@n9, InnerAngle) {materialFunction=/FortUILibrary/MaterialFunctions/Shapes/SDFs/MF_SDF_RegularPolygon}.SDF) {materialFunction=&MF_UI_AntiAliasing}.Inner) {name=Triangles, nodeColor=<rgba 0.3594 1 0 1>}
 ]
}
